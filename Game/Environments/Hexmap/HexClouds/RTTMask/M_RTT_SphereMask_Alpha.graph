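# Unlit alpha-composite material for the hexmap clouds' RTT mask, forming a soft circular falloff that is presumably drawn into a render target.
# A SphereMask compares texture coordinates to the UV center, with a radius reaching the edges and a constant hardness.
# A Power node squares the mask to tighten the gradient toward the center.
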Material /Game/Environments/Hexmap/HexClouds/RTTMask/M_RTT_SphereMask_Alpha
graph {
 properties {blendMode=BLEND_AlphaComposite, pixelDepthOffsetMode=PDOM_Legacy, shadingModel=MSM_Unlit, usedShadingModels=Unlit}
 nodes [
  Constant()
  n1 = Constant() {r=0.5}
  Power(Base=SphereMask(A=TextureCoordinate(), B=@n1, Radius=@n1, Hardness=Constant()), Exp=Constant() {r=2})
 ]
}
